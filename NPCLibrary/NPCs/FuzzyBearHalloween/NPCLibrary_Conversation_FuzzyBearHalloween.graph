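# Conversation database for the FuzzyBearHalloween NPC, a Halloween bear medic specialist, defining its greeting and dialogue links.
# The entry point leads to one speech node that has a default line plus contextual lines gated by first-ever conversation, per-match conversation count, or the player wearing the PinkBear outfit.
# Priority curves rank the first-meeting and outfit reactions.
# The speech node then links out to the shared NPCLibrary Quests and BasicServices conversations.
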ConversationDatabase /NPCLibrary/NPCs/FuzzyBearHalloween/NPCLibrary_Conversation_FuzzyBearHalloween
entry_points [{tag=Conversation.Entry.FuzzyBearHalloween, destinations [#1]}]
nodes [
 ConversationEntryPointNode_7 {
  id=#1
  class=ConversationEntryPointNode
  outputs [#2]
  properties {entryTag {tagName=Conversation.Entry.FuzzyBearHalloween}}
 }
 Speech_C_4 {
  id=#2
  class=Speech_C
  outputs [#3, #4]
  properties {
   nodeName=""
   generalConfig {
    defaultMessage=This bear's got the scares.
    contextualMessages [
     {
      message="It's my favorite time of the year!"
      contextRequirements [{requirement {class=FortControllerRequirement_IsFirstEverConversationWithNPC}}]
      priority {curve {rowName=Tandem.Speech.Priority.FirstEverInteraction}}
     }
     {
      message=This bear's got the scares.
      contextRequirements [{requirement {
         class=FortControllerRequirement_CheckNPCMatchConversationCount
         properties {nPCUniqueIdTag {tagName=AISpawnerData.Type.Specialist.Medic.FuzzyBearHalloween}, requiredConversationCountConfig {requirement=GreaterThan}}
        }}]
      priority {curve {curveTable=null, rowName=None}}
     }
     {
      message="Nothin' I love more than Fortnitemares!"
      contextRequirements [{requirement {
         class=FortControllerRequirement_CheckNPCMatchConversationCount
         properties {nPCUniqueIdTag {tagName=AISpawnerData.Type.Specialist.Medic.FuzzyBearHalloween}, requiredConversationCountConfig {requirement=GreaterThan}}
        }}]
      priority {curve {curveTable=null, rowName=None}}
     }
     {
      message="Beware the bear!"
      contextRequirements [{requirement {
         class=FortControllerRequirement_CheckNPCMatchConversationCount
         properties {nPCUniqueIdTag {tagName=AISpawnerData.Type.Specialist.Medic.FuzzyBearHalloween}, requiredConversationCountConfig {requirement=GreaterThan}}
        }}]
      priority {curve {curveTable=null, rowName=None}}
     }
     {
      message="You look... pink!"
      contextRequirements [{requirement {
         class=FortControllerRequirement_HasCID
         properties {allowedCIds [/Game/Athena/Items/Cosmetics/Characters/CID_069_Athena_Commando_F_PinkBear]}
        }}]
      priority {curve {rowName=Tandem.Speech.Priority.OutfitReaction}}
     }
    ]
   }
  }
 }
 ConversationLinkNode_0 {id=#4, class=ConversationLinkNode, properties {remoteEntryTag {tagName=Conversation.Entry.BasicServices}}}
 ConversationLinkNode_1 {id=#3, class=ConversationLinkNode, properties {remoteEntryTag {tagName=Conversation.Entry.NPCLibrary.Quests}}}
]
